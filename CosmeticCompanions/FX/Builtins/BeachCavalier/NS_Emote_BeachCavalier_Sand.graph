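# Niagara sand effect for the BeachCavalier companion emote, layering six sprite emitters of soil, dirt, smoke and dust.
# Soil, Dirt, Smoke002 and PollenDust001 define two burst groups, with spawn-group masks varying offsets.
# Delayed HandTapDust fires three bursts, and later BucketPullDust spawns from the mesh's bucket socket.
# Time of day drives tint and opacity, while front-end and ForwardDot inputs adjust some layers.
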NiagaraSystem /CosmeticCompanions/FX/Builtins/BeachCavalier/NS_Emote_BeachCavalier_Sand
dependencies {
 usedDataInterfaces [
  &NiagaraDataInterfaceArrayFloat
  &NiagaraDataInterfaceCurve
  &NiagaraDataInterfaceVector2DCurve
  &NiagaraDataInterfaceColorCurve
  &NiagaraDataInterfacePlatformSet
  &NiagaraDataInterfaceMeshRendererInfo
  &NiagaraDataInterfaceSocketReader
  &NiagaraDataInterfaceVectorCurve
 ]
 usedDynamicInputs [
  &SelectFloatFromArray
  &SpawnGroup
  &RandomRangeInt
 ]
 usedModules [&SystemState]
}
emitters {
 BucketPullDust {
  emitterUpdateScript [
   &EmitterState(Life Cycle Mode=::NewEnumerator1, Loop Behavior=::NewEnumerator1, Loop Delay=3.275, Loop Duration=0.8, UseLoopDelay=true)
   &SpawnBurst_Instantaneous(Use Spawn Probability=false, UseLoopCountLimit=false)
   &SpawnBurst_Instantaneous(Spawn Count=2, Spawn Time=0.08, Use Spawn Probability=false, UseLoopCountLimit=false) {module=SpawnBurst_Instantaneous001}
   &SpawnRate(
    SpawnRate=&FloatFromCurve(
     FloatCurve=DataInterface() {propertyValues {Curve {keys [<at 0 1.6 RCIM_Cubic>, {arriveTangent=-0.6, interpMode=RCIM_Cubic, leaveTangent=-0.6, time=0.2, value=1}, <at 1 1>]}}}
     CurveIndex=$Emitter.NormalizedLoopAge
     Scale Curve=36
    )
    Use Spawn Probability=false
   )
  ]
  particleSpawnScript [
   &InitializeParticle(
    Color=<rgba 3 3 3 1>
    Lifetime=&UniformRangedFloat(Minimum=0.5, Maximum=0.7)
    Position Offset=<0 0 45>
    Sprite Rotation Mode=::NewEnumerator2
    Sprite Size Mode=::NewEnumerator1
    Uniform Sprite Size Max=95
    Uniform Sprite Size Min=70
    UsePositionOffset=true
    Write Material Random=true
   )
   &DynamicMaterialParameters(
    Index 0 Param 1=&RandomRangeFloat(Minimum=-1, Maximum=1)
    Index 0 Param 2=&RandomRangeFloat(Minimum=0, Maximum=32)
    Module.Index0Float0Write=true
    Module.Index0Float1Write=true
    Module.Index0Float2Write=true
    Module.Index0Float3Write=true
   )
   &CameraOffset(Camera Offset Amount=20)
   &AddVelocity(Velocity=&RandomRangeVector(Minimum=<0 0 -300>, Maximum=<1 1 -100>))
   &SocketLocation(
    SamplePosOffset=true
    Sampled Position Offset=<0 0 -5>
    Socket Reader=DataInterface() {propertyValues {
      AttachComponentClass=null
      AttachComponentTag=None
      EditorPreviewAsset=/CosmeticCompanions/Assets/Props/Emotes/BeachCavalier/Meshes/SKM_BeachCavalier
      SourceActorRef=null
      SourceAsset=null
      SourceMode=Default
      bRequireCurrentFrameData=true
      bUpdateSocketsPerFrame=true
      FilteredSockets [bucket]
      ObjectParameterBinding {parameter {name=None, typeDefHandle {registeredTypeIndex=1}, varData []}}
     }}
    Socket Sampling Mode=::NewEnumerator3
    VelocitySpeedLimit=false
   )
   &ShapeLocation(MaskSpawnGroup=false, Override Seed=false, Sphere Radius=5, Sphere Surface Distribution=1, UseOwnerScale=true)
  ]
  particleUpdateScript [
   &ParticleState()
   &AccelerationForce(Acceleration=&MakeVector(X=0, Y=0, Z=150))
   &ScaleVelocity(Velocity Scale=&Multiply_VectorByFloat(
     Vector=&VectorFromCurve(VectorCurve=DataInterface() {propertyValues {
        XCurve {keys [{arriveTangent=3.1805, interpMode=RCIM_Cubic, leaveTangent=3.1805, tangentMode=RCTM_User, time=0, value=0}, <at 1 1.4>]}
        YCurve {keys [{arriveTangent=3.1805, interpMode=RCIM_Cubic, leaveTangent=3.1805, tangentMode=RCTM_User, time=0, value=0}, <at 1 1.4>]}
        ZCurve {keys [{arriveTangent=2.4373, interpMode=RCIM_Cubic, leaveTangent=2.4373, tangentMode=RCTM_User, time=0, value=0}, <at 1 0.7714>]}
       }})
     Float=&FN_FrontEndFloatModifier(
      Gameplay=&ClampFloat(Float=$User.ForwardDot, Min=&FN_FrontEndFloatModifier(Gameplay=0.5, Front End=0.8))
      Front End=1
     )
    ))
   &Drag(Drag=&RandomRangeFloat(Minimum=2, Maximum=0), Use Linear Drag=true, Use Rotational Drag=true)
   &SolveForcesAndVelocity(Clamp Velocity=false, Limit Acceleration=false)
   &ScaleColor(
    Scale Alpha=&FloatFromCurve(
     FloatCurve=DataInterface() {propertyValues {Curve {keys [
         {interpMode=RCIM_Cubic, tangentMode=RCTM_User, time=0, value=0}
         <at 0.3224 1 RCIM_Cubic>
         {interpMode=RCIM_Cubic, leaveTangent=-3.8631, time=1, value=0}
        ]}}}
     Scale Curve=&Multiply_Float(A=&FNTimeOfDayScalar(Night Scale=0.6, Morning Scale=0.8, Evening Scale=0.8), B=0.3)
    )
    Scale RGB=&MakeVectorFromLinearColorRGB(Color=&ColorFromCurve(
      ColorCurve=DataInterface() {propertyValues {
        AlphaCurve {keys [<at 0.13 1>,<at 0.173 1>,<at 0.25 1>,<at 0.375 1>,<at 0.5 1>,<at 0.625 1>,<at 0.75 1>,<at 0.873 1>,<at 0.9078 1>]}
        BlueCurve {keys [<at 0.0514 0.1497>,<at 0.1202 0.2248>,<at 0.25 0.2724>,<at 0.5 0.2754>,<at 0.7338 0.4077>,<at 0.8202 0.2971>,<at 0.8836 0.4589>,<at 0.9613 0.1497>]}
        GreenCurve {keys [<at 0.0514 0.2159>,<at 0.1202 0.3822>,<at 0.25 0.4184>,<at 0.5 0.4897>,<at 0.7338 0.6015>,<at 0.8202 0.6157>,<at 0.8836 0.5652>,<at 0.9613 0.2159>]}
        RedCurve {keys [<at 0.0514 0.2448>,<at 0.1202 0.7344>,<at 0.25 0.7344>,<at 0.5 0.7344>,<at 0.7338 0.7344>,<at 0.8202 0.9855>,<at 0.8836 0.6846>,<at 0.9613 0.2448>]}
       }}
      CurveIndex=Linked() {linkedVariable {name=NPC.FortniteNPC.TimeOfDay_0-1, type=NiagaraFloat}}
     ))
    ScaleA=true
    ScaleRGB=true
   )
   &ScaleSpriteSize(
    Uniform Curve Scale=&FloatFromCurve(
     FloatCurve=DataInterface() {propertyValues {Curve {
        defaultValue=1
        keys [
         {arriveTangent=1.8398, interpMode=RCIM_Cubic, leaveTangent=1.8398, tangentMode=RCTM_User, time=2.1, value=0.274}
         {interpMode=RCIM_Cubic, tangentMode=RCTM_User, time=2.6, value=1}
        ]
       }}}
     CurveIndex=$Emitter.Age
     Scale Curve=0.6
    )
    Uniform Curve Sprite Scale=DataInterface() {propertyValues {Curve {keys [
        {arriveTangent=0.6664, interpMode=RCIM_Cubic, leaveTangent=0.6664, tangentMode=RCTM_User, time=0, value=1}
        {interpMode=RCIM_Cubic, leaveTangent=-0.4663, time=0.4287, value=1.1262}
       ]}}}
   )
  ]
  rendererClasses [&NiagaraSpriteRendererProperties]
  renderers [{rendererClass=&NiagaraSpriteRendererProperties}]
 }
 Dirt {
  emitterUpdateScript [
   &EmitterState(Inactive Response=::NewEnumerator1, Life Cycle Mode=::NewEnumerator1, Loop Behavior=::NewEnumerator1, Loop Duration=1, UseLoopDelay=false)
   &SpawnBurst_Instantaneous(Spawn Count=15, Use Spawn Probability=false, UseLoopCountLimit=false)
   &SpawnBurst_Instantaneous(Spawn Count=6, Spawn Group=1, Spawn Time=1.204, Use Spawn Probability=false, UseLoopCountLimit=false) {module=SpawnBurst_Instantaneous001}
  ]
  particleSpawnScript [
   &InitializeParticle(
    Lifetime Max=1.5
    Lifetime Min=0.5
    Lifetime Mode=::NewEnumerator1
    Position Offset=&MaskVectorBySpawnGroup(Masked Vector=<0 0 0>, Passthrough Vector=<-8 -10 0>)
    Sprite Size Max=<1 4>
    Sprite Size Min=<0.5 2>
    Sprite Size Mode=::NewEnumerator2
    UsePositionOffset=true
    Write Material Random=true
   )
   &ShapeLocation(
    Handle Radius=10
    Large Radius=15
    MaskSpawnGroup=false
    Non Uniform Scale=<1 1 0.5>
    Offset Mode=::NewEnumerator0
    Override Seed=false
    Shape Primitive=::NewEnumerator3
    UseOwnerScale=true
   )
   &AddVelocity(
    Module.Use Falloff Distance=false
    Origin Offset=&MaskVectorBySpawnGroup(Masked Vector=<0 0 -10>, Passthrough Vector=<-8 -10 -10>)
    Override Seed=false
    Velocity Mode=::NewEnumerator1
    Velocity Speed=&RandomRangeFloat(Minimum=150, Maximum=200)
   )
  ]
  particleUpdateScript [
   &ParticleState()
   &ScaleSpriteSize(Uniform Curve Sprite Scale=DataInterface() {propertyValues {Curve {keys [
        {arriveTangent=8.2127, interpMode=RCIM_Cubic, leaveTangent=8.2127, tangentMode=RCTM_User, time=0, value=0.5}
        {
         arriveTangent=4.2448
         arriveTangentWeight=0.232
         interpMode=RCIM_Cubic
         leaveTangent=4.2448
         leaveTangentWeight=2.2538
         tangentMode=RCTM_User
         time=0.07
         value=1
        }
        {arriveTangent=-2.5194, interpMode=RCIM_Cubic, leaveTangent=-2.5194, tangentMode=RCTM_User, time=0.6, value=0.8}
        {arriveTangent=-0.2055, interpMode=RCIM_Cubic, leaveTangent=-0.2055, tangentMode=RCTM_User, time=1, value=0}
       ]}}})
   &DynamicMaterialParameters(
    Index 0 Param 1=&FloatFromCurve(
     FloatCurve=DataInterface() {propertyValues {Curve {keys [
         {arriveTangent=-5.8452, interpMode=RCIM_Cubic, time=0, value=1}
         {arriveTangentWeight=0.0434, interpMode=RCIM_Cubic, leaveTangentWeight=0.217, tangentMode=RCTM_User, time=0.2, value=0}
         {arriveTangent=3.4164, interpMode=RCIM_Cubic, leaveTangent=3.4164, tangentMode=RCTM_User, time=0.8, value=0.7}
         <at 1 0 RCIM_Cubic>
        ]}}}
     Scale Curve=3
    )
    Index 0 Param 3=1
    Index 0 Param 4=&FloatFromCurve(
     FloatCurve=DataInterface() {propertyValues {Curve {keys [
         {interpMode=RCIM_Cubic, leaveTangent=-1.5195, tangentMode=RCTM_Break, time=0, value=1}
         {arriveTangent=0.0067, interpMode=RCIM_Cubic, tangentMode=RCTM_Break, time=1, value=0}
        ]}}}
     Scale Curve=&RandomRangeFloat(Minimum=10, Maximum=30)
    )
    Module.Param0WriteEnabled=true
    Module.Param1WriteEnabled=true
    Module.Param2WriteEnabled=true
    Module.Param3WriteEnabled=true
   )
   &Drag(Drag=&RandomRangeFloat(Minimum=1, Maximum=2), Use Linear Drag=true, Use Rotational Drag=true)
   &GravityForce(Gravity=&RandomRangeVector(Minimum=<0 0 -300>, Maximum=<0 0 -600>))
   &SolveForcesAndVelocity(Clamp Velocity=false, Limit Acceleration=false)
   &ScaleColor(
    Scale Alpha=&Lerp_Float(
     A=0
     B=1
     Alpha=&FloatFromCurve(
      FloatCurve=DataInterface() {propertyValues {Curve {keys [
          {interpMode=RCIM_Cubic, tangentMode=RCTM_User, time=0, value=1}
          {interpMode=RCIM_Cubic, tangentMode=RCTM_User, time=0.5, value=1}
          {tangentMode=RCTM_User, time=1, value=0}
         ]}}}
      Scale Curve=&Multiply_Float(A=1, B=&FNTimeOfDayScalar(Night Scale=0.6, Morning Scale=0.9, Evening Scale=0.9))
     )
    )
    Scale RGB=&MakeVectorFromLinearColorRGB(Color=&FNTimeOfDayColor(
      Night Color=&RandomRangeLinearColor(Minimum=<rgba 0.5272 0.4166 0.3591 1>, Maximum=<rgba 0.6152 0.4984 0.4446 1>)
      Morning Color=&RandomRangeLinearColor(Minimum=<rgba 0.9253 0.5424 0.1419 1>, Maximum=<rgba 0.8466 0.4429 0.1102 1>)
      Day Color=&RandomRangeLinearColor(Minimum=<rgba 0.805 0.6212 0.429 1>, Maximum=<rgba 0.7818 0.6471 0.5361 1>)
      Evening Color=&RandomRangeLinearColor(Minimum=<rgba 0.5271 0.3564 0.1779 1>, Maximum=<rgba 0.5271 0.3864 0.2705 1>)
     ))
    ScaleA=true
    ScaleRGB=true
   )
   &CameraOffset(Camera Offset Amount=10)
  ]
  rendererClasses [&NiagaraSpriteRendererProperties]
  renderers [{rendererClass=&NiagaraSpriteRendererProperties}]
 }
 HandTapDust {
  emitterUpdateScript [
   &EmitterState(
    Inactive Response=::NewEnumerator1
    Life Cycle Mode=::NewEnumerator1
    Loop Behavior=::NewEnumerator1
    Loop Delay=1.875
    Loop Duration=1.4
    UseLoopDelay=true
   )
   &SpawnBurst_Instantaneous(Spawn Count=6, Use Spawn Probability=false, UseLoopCountLimit=false)
   &SpawnBurst_Instantaneous(Spawn Count=4, Spawn Time=0.35, Use Spawn Probability=false, UseLoopCountLimit=false) {module=SpawnBurst_Instantaneous001}
   &SpawnBurst_Instantaneous(Spawn Count=7, Spawn Time=0.7, Use Spawn Probability=false, UseLoopCountLimit=false) {module=SpawnBurst_Instantaneous002}
  ]
  particleSpawnScript [
   &InitializeParticle(
    Lifetime Max=1
    Lifetime Min=0.4
    Lifetime Mode=::NewEnumerator1
    Position Offset=<-9 -8 30>
    Sprite Size Mode=::NewEnumerator1
    Uniform Sprite Size Max=32
    Uniform Sprite Size Min=25
    UsePositionOffset=true
    Write Material Random=true
   )
   &DynamicMaterialParameters(
    Index 0 Param 1=&RandomRangeFloat(Minimum=-1, Maximum=1)
    Index 0 Param 2=&RandomRangeFloat(Minimum=0, Maximum=32)
    Module.Index0Float0Write=true
    Module.Index0Float1Write=true
    Module.Index0Float2Write=true
    Module.Index0Float3Write=true
   )
   &ShapeLocation(
    Handle Radius=1
    Large Radius=4
    MaskSpawnGroup=false
    Non Uniform Scale=<1 1 0.5>
    Offset Mode=::NewEnumerator0
    Override Seed=false
    Shape Primitive=::NewEnumerator3
    UseOwnerScale=true
   )
   &AddVelocity(
    Module.Use Falloff Distance=false
    Origin Offset=<-9 -8 30>
    Override Seed=false
    Velocity Mode=::NewEnumerator1
    Velocity Speed=&RandomRangeFloat(Minimum=150, Maximum=75)
   )
  ]
  particleUpdateScript [
   &ParticleState()
   &Drag(Drag=&RandomRangeFloat(Minimum=5, Maximum=10), Use Linear Drag=true, Use Rotational Drag=true)
   &GravityForce(Gravity=&RandomRangeVector(Minimum=<0 0 -300>, Maximum=<0 0 -600>))
   &SolveForcesAndVelocity(Clamp Velocity=false, Limit Acceleration=false)
   &ScaleColor(
    Scale Alpha=&Lerp_Float(
     A=0
     B=1
     Alpha=&FloatFromCurve(
      FloatCurve=DataInterface() {propertyValues {Curve {keys [{interpMode=RCIM_Cubic, tangentMode=RCTM_User, time=0, value=1}, {tangentMode=RCTM_User, time=1, value=0}]}}}
      Scale Curve=&Multiply_Float(A=1, B=&FNTimeOfDayScalar(Night Scale=0.6, Morning Scale=0.9, Evening Scale=0.9))
     )
    )
    Scale RGB=&MakeVectorFromLinearColorRGB(Color=&ColorFromCurve(
      ColorCurve=DataInterface() {propertyValues {
        AlphaCurve {keys [<at 0.13 1>,<at 0.173 1>,<at 0.25 1>,<at 0.375 1>,<at 0.5 1>,<at 0.625 1>,<at 0.75 1>,<at 0.873 1>,<at 0.9078 1>]}
        BlueCurve {keys [<at 0.0514 0.4233>,<at 0.1202 0.2248>,<at 0.25 0.2724>,<at 0.5 0.2754>,<at 0.7338 0.4077>,<at 0.8202 0.2971>,<at 0.8836 0.4589>,<at 0.9613 0.4197>]}
        GreenCurve {keys [<at 0.0514 0.5117>,<at 0.1202 0.3822>,<at 0.25 0.4184>,<at 0.5 0.4897>,<at 0.7338 0.6015>,<at 0.8202 0.6157>,<at 0.8836 0.5652>,<at 0.9613 0.5074>]}
        RedCurve {keys [<at 0.0514 0.5503>,<at 0.1202 0.7344>,<at 0.25 0.7344>,<at 0.5 0.7344>,<at 0.7338 0.7344>,<at 0.8202 0.9855>,<at 0.8836 0.6846>,<at 0.9613 0.5457>]}
       }}
      CurveIndex=Linked() {linkedVariable {name=NPC.FortniteNPC.TimeOfDay_0-1, type=NiagaraFloat}}
     ))
    ScaleA=true
    ScaleRGB=true
   )
  ]
  rendererClasses [&NiagaraSpriteRendererProperties]
  renderers [{rendererClass=&NiagaraSpriteRendererProperties}]
 }
 PollenDust001 {
  emitterUpdateScript [
   &EmitterState(Life Cycle Mode=::NewEnumerator1, Loop Behavior=::NewEnumerator1, Loop Duration=0.5, UseLoopDelay=true)
   &SpawnBurst_Instantaneous(Spawn Count=4, Use Spawn Probability=false, UseLoopCountLimit=false)
   &SpawnBurst_Instantaneous(Spawn Count=4, Spawn Group=1, Spawn Time=1.204, Use Spawn Probability=false, UseLoopCountLimit=false) {module=SpawnBurst_Instantaneous001}
  ]
  particleSpawnScript [
   &InitializeParticle(
    Color=<rgba 3 3 3 1>
    Lifetime=&UniformRangedFloat(Minimum=0.5, Maximum=0.7)
    Position Offset=&MaskVectorBySpawnGroup(Masked Vector=<0 0 27>, Passthrough Vector=<-8 -10 27>)
    Sprite Rotation Mode=::NewEnumerator2
    Sprite Size Mode=::NewEnumerator1
    Uniform Sprite Size Max=80
    Uniform Sprite Size Min=100
    UsePositionOffset=true
    Write Material Random=true
   )
   &DynamicMaterialParameters(
    Index 0 Param 1=&RandomRangeFloat(Minimum=-1, Maximum=1)
    Index 0 Param 2=&RandomRangeFloat(Minimum=0, Maximum=32)
    Module.Index0Float0Write=true
    Module.Index0Float1Write=true
    Module.Index0Float2Write=true
    Module.Index0Float3Write=true
   )
   &CameraOffset(Camera Offset Amount=20)
   &AddVelocity(Velocity=&RandomRangeVector(Minimum=<0 0 -300>, Maximum=<1 1 -100>))
   &ShapeLocation(Handle Radius=5, Large Radius=10, MaskSpawnGroup=false, Override Seed=false, Shape Primitive=::NewEnumerator3, UseOwnerScale=true)
  ]
  particleUpdateScript [
   &ParticleState()
   &AccelerationForce(Acceleration=&MakeVector(X=0, Y=0, Z=150))
   &ScaleVelocity(Velocity Scale=&Multiply_VectorByFloat(
     Vector=&VectorFromCurve(VectorCurve=DataInterface() {propertyValues {
        XCurve {keys [{arriveTangent=3.1805, interpMode=RCIM_Cubic, leaveTangent=3.1805, tangentMode=RCTM_User, time=0, value=0}, <at 1 1.4>]}
        YCurve {keys [{arriveTangent=3.1805, interpMode=RCIM_Cubic, leaveTangent=3.1805, tangentMode=RCTM_User, time=0, value=0}, <at 1 1.4>]}
        ZCurve {keys [{arriveTangent=2.4373, interpMode=RCIM_Cubic, leaveTangent=2.4373, tangentMode=RCTM_User, time=0, value=0}, <at 1 0.7714>]}
       }})
     Float=&FN_FrontEndFloatModifier(
      Gameplay=&ClampFloat(Float=$User.ForwardDot, Min=&FN_FrontEndFloatModifier(Gameplay=0.5, Front End=0.8))
      Front End=1
     )
    ))
   &Drag(Drag=&RandomRangeFloat(Minimum=2, Maximum=0), Use Linear Drag=true, Use Rotational Drag=true)
   &SolveForcesAndVelocity(Clamp Velocity=false, Limit Acceleration=false)
   &ScaleColor(
    Scale Alpha=&FloatFromCurve(
     FloatCurve=DataInterface() {propertyValues {Curve {keys [
         {interpMode=RCIM_Cubic, tangentMode=RCTM_User, time=0, value=0}
         <at 0.3224 1 RCIM_Cubic>
         {interpMode=RCIM_Cubic, leaveTangent=-3.8631, time=1, value=0}
        ]}}}
     Scale Curve=&Multiply_Float(
      A=&FNTimeOfDayScalar(Night Scale=0.6, Morning Scale=0.8, Day Scale=0.7, Evening Scale=0.8)
      B=0.3
     )
    )
    Scale RGB=&MakeVectorFromLinearColorRGB(Color=&ColorFromCurve(
      ColorCurve=DataInterface() {propertyValues {
        AlphaCurve {keys [<at 0.13 1>,<at 0.173 1>,<at 0.25 1>,<at 0.375 1>,<at 0.5 1>,<at 0.625 1>,<at 0.75 1>,<at 0.873 1>,<at 0.9078 1>]}
        BlueCurve {keys [<at 0.0514 0.1987>,<at 0.1202 0.2248>,<at 0.25 0.2724>,<at 0.5 0.2754>,<at 0.7338 0.4077>,<at 0.8202 0.2971>,<at 0.8836 0.4589>,<at 0.9613 0.1987>]}
        GreenCurve {keys [<at 0.0514 0.2662>,<at 0.1202 0.3822>,<at 0.25 0.4184>,<at 0.5 0.4897>,<at 0.7338 0.6015>,<at 0.8202 0.6157>,<at 0.8836 0.5652>,<at 0.9613 0.2662>]}
        RedCurve {keys [<at 0.0514 0.2957>,<at 0.1202 0.7344>,<at 0.25 0.7344>,<at 0.5 0.7344>,<at 0.7338 0.7344>,<at 0.8202 0.9855>,<at 0.8836 0.6846>,<at 0.9613 0.2957>]}
       }}
      CurveIndex=Linked() {linkedVariable {name=NPC.FortniteNPC.TimeOfDay_0-1, type=NiagaraFloat}}
     ))
    ScaleA=true
    ScaleRGB=true
   )
   &ScaleSpriteSize(Uniform Curve Sprite Scale=DataInterface() {propertyValues {Curve {keys [
        {arriveTangent=0.6664, interpMode=RCIM_Cubic, leaveTangent=0.6664, tangentMode=RCTM_User, time=0, value=0.777}
        {interpMode=RCIM_Cubic, leaveTangent=-0.4663, time=1, value=1.1262}
       ]}}})
  ]
  rendererClasses [&NiagaraSpriteRendererProperties]
  renderers [{rendererClass=&NiagaraSpriteRendererProperties}]
 }
 Smoke002 {
  emitterUpdateScript [
   &EmitterState(Inactive Response=::NewEnumerator1, Life Cycle Mode=::NewEnumerator1, Loop Behavior=::NewEnumerator1, Loop Duration=3, UseLoopDelay=false)
   &SpawnBurst_Instantaneous(Spawn Count=9, Use Spawn Probability=true, UseLoopCountLimit=false)
   &SpawnBurst_Instantaneous(Spawn Count=12, Spawn Group=1, Spawn Time=1.204, Use Spawn Probability=true, UseLoopCountLimit=false) {module=SpawnBurst_Instantaneous001}
  ]
  particleSpawnScript [
   &InitializeParticle(
    Color=&FNTimeOfDayColor(
     Night Color=&RandomRangeLinearColor(Minimum=<rgba 0.9624 0.9146 0.8605 1>, Maximum=<rgba 0.9253 0.8611 0.8016 0.5>)
     Morning Color=&RandomRangeLinearColor(Minimum=<rgba 0.9624 0.6032 0.1966 1>, Maximum=<rgba 0.8605 0.4906 0.1479 0.5>)
     Day Color=&RandomRangeLinearColor(Minimum=<rgba 0.8466 0.6373 0.4003 1>, Maximum=<rgba 0.7402 0.4771 0.2334 0.5>)
     Evening Color=&RandomRangeLinearColor(Minimum=<rgba 0.6198 0.4665 0.293 1>, Maximum=<rgba 0.3605 0.2324 0.1137 0.5>)
    )
    Lifetime Max=0.6
    Lifetime Min=0.5
    Lifetime Mode=::NewEnumerator1
    Position Offset=&MaskVectorBySpawnGroup(Masked Vector=<0 0 0>, Passthrough Vector=<-8 -10 0>)
    Sprite Rotation Mode=::NewEnumerator2
    Sprite Size Mode=::NewEnumerator3
    Uniform Sprite Size=&MaskFloatBySpawnGroup(
     Masked Float=&RandomRangeFloat(Minimum=60, Maximum=80)
     Passthrough Float=&RandomRangeFloat(Minimum=45, Maximum=70)
    )
    UsePositionOffset=true
    Write Material Random=true
   )
   /SetVariables(Particles.InitSpriteRotRate=&UniformRangedFloat(Minimum=-3, Maximum=3)) {isSetParameters=true}
   &ShapeLocation(
    Handle Radius=2
    Large Radius=7
    MaskSpawnGroup=false
    Non Uniform Scale=<2 2 1>
    Offset Mode=::NewEnumerator0
    Override Seed=false
    Shape Primitive=::NewEnumerator3
    UseOwnerScale=true
   )
   &AddVelocity(
    Module.Use Falloff Distance=false
    Origin Offset=&MaskVectorBySpawnGroup(Masked Vector=<0 0 -5>, Passthrough Vector=<-8 -10 -5>)
    Override Seed=false
    Velocity Mode=::NewEnumerator1
    Velocity Speed=&MaskFloatBySpawnGroup(
     Masked Float=&RandomRangeFloat(Minimum=20, Maximum=50)
     Passthrough Float=&RandomRangeFloat(Minimum=10, Maximum=30)
    )
   )
   &AddVelocity(Velocity=&MakeVector(X=0, Y=0, Z=&MaskFloatBySpawnGroup(Masked Float=50, Passthrough Float=15))) {module=AddVelocity001}
   &CameraOffset(Camera Offset Amount=10)
  ]
  particleUpdateScript [
   &ParticleState()
   &SpriteRotationRate(Rotation Rate=&Multiply_Float(
     A=$Particles.InitSpriteRotRate
     B=&FloatFromCurve(
      FloatCurve=DataInterface() {propertyValues {Curve {keys [
          {arriveTangent=-4.0107, interpMode=RCIM_Cubic, leaveTangent=-4.0107, tangentMode=RCTM_User, time=-0.0046, value=1.0113}
          {arriveTangent=-0.2496, interpMode=RCIM_Cubic, leaveTangent=-0.2496, tangentMode=RCTM_User, time=0.5, value=0.16}
         ]}}}
      CurveIndex=$Emitter.Age
      Scale Curve=150
     )
    ))
   &ScaleColor(
    Scale Alpha=&Multiply_Float(
     A=&MakeCustomFloatFromBool(
      Bool=$User.bIsFrontEndPreview
      True Float=0.5
      False Float=&FN_PickFloatDependOnPlatform(TargetPlatformActive=0.6, TargetPlatformNoActive=0.5)
     )
     B=&FloatFromCurve(
      FloatCurve=DataInterface() {propertyValues {Curve {
         defaultValue=1
         keys [
          {arriveTangent=6.2609, interpMode=RCIM_Cubic, leaveTangent=6.2609, tangentMode=RCTM_User, time=0, value=0}
          {interpMode=RCIM_Cubic, tangentMode=RCTM_User, time=0.5, value=1}
          {arriveTangent=-6.1916, tangentMode=RCTM_User, time=1, value=0}
         ]
        }}}
      Scale Curve=&MaskFloatBySpawnGroup(Masked Float=0.4, Passthrough Float=0.325)
     )
    )
    Scale RGB=&MakeVectorFromLinearColorRGB(Color=<rgba 1 0.8741 0.7784 1>)
    ScaleA=true
    ScaleRGB=true
   ) {module=ScaleColor001}
   /Niagara/Modules/Update/Forces/CurlNoiseForce(Noise Frequency=0.1, Noise Strength=200, Pan Noise=true, Pan Noise Field=<0 0 1>)
   &GravityForce(Gravity=&MakeVector(X=0, Y=0, Z=&MaskFloatBySpawnGroup(Masked Float=-300, Passthrough Float=-125)))
   &Drag(Drag=2, Use Linear Drag=true, Use Rotational Drag=false)
   &SolveForcesAndVelocity(Clamp Velocity=false, Limit Acceleration=false)
   &ScaleSpriteSize(Scale Factor=&Vector2DFromFloat(Value=&FloatFromCurve(
      FloatCurve=DataInterface() {propertyValues {Curve {keys [
          {arriveTangent=8.1876, interpMode=RCIM_Cubic, leaveTangent=5.3162, tangentMode=RCTM_User, time=0, value=0.15}
          {arriveTangent=1.0025, interpMode=RCIM_Cubic, leaveTangent=1.0025, tangentMode=RCTM_User, time=0.166, value=0.7682}
          {arriveTangent=0.2754, interpMode=RCIM_Cubic, leaveTangent=0.2754, tangentMode=RCTM_User, time=1, value=1.2}
         ]}}}
      Scale Curve=&MaskFloatBySpawnGroup(Masked Float=0.4, Passthrough Float=0.325)
     )))
   &DynamicMaterialParameters(
    Index 0 Param 1=&Add_Float(
     A=&RandomRangeFloat(Minimum=0, Maximum=1)
     B=&FloatFromCurve(
      FloatCurve=DataInterface() {propertyValues {Curve {keys [
          {arriveTangent=2.7533, interpMode=RCIM_Cubic, leaveTangent=1.8958, tangentMode=RCTM_User, time=0, value=1}
          {arriveTangent=0.8292, interpMode=RCIM_Cubic, leaveTangent=0.8292, tangentMode=RCTM_User, time=0.321, value=1.3453}
          {arriveTangent=0.0529, interpMode=RCIM_Cubic, leaveTangent=0.2021, tangentMode=RCTM_User, time=1, value=1.5}
         ]}}}
      CurveIndex=&Add_Float(A=&Multiply_Float(A=$Particles.Initial.MaterialRandom, B=0.5), B=$Particles.NormalizedAge)
     )
    )
    Index 0 Param 2=&FloatFromCurve(FloatCurve=DataInterface() {propertyValues {Curve {keys [
         {arriveTangent=0.0453, interpMode=RCIM_Cubic, tangentMode=RCTM_Break, time=0, value=-1}
         {arriveTangent=0.0604, interpMode=RCIM_Cubic, leaveTangent=0.0604, tangentMode=RCTM_User, time=1, value=-1}
        ]}}})
    Index 0 Param 3=&FloatFromCurve(
     FloatCurve=DataInterface() {propertyValues {Curve {keys [
         {arriveTangent=-1.5217, interpMode=RCIM_Cubic, leaveTangent=-1.5217, tangentMode=RCTM_User, time=0, value=0.2}
         {arriveTangent=0.1017, interpMode=RCIM_Cubic, leaveTangent=0.2723, tangentMode=RCTM_Break, time=0.2, value=0}
        ]}}}
     Scale Curve=0.3
    )
    Index 0 Param 4=&FloatFromCurve(FloatCurve=DataInterface() {propertyValues {Curve {keys [<at 0.5 1>,<at 1 0>]}}})
    Index 1 Param 1=&FloatFromCurve(
     FloatCurve=DataInterface() {propertyValues {Curve {keys [
         {interpMode=RCIM_Cubic, leaveTangent=4.7088, tangentMode=RCTM_Break, time=0, value=0}
         {arriveTangent=4.2257, interpMode=RCIM_Cubic, leaveTangent=4.2257, tangentMode=RCTM_User, time=0.25, value=1}
        ]}}}
     Scale Curve=&RandomRangeFloat(Minimum=1.5, Maximum=1.5)
    )
    Index 1 Param 2=&FloatFromCurve(
     FloatCurve=DataInterface() {propertyValues {Curve {keys [<at 0 0.2>, {leaveTangent=0.9066, time=1, value=2}]}}}
     Scale Curve=&RandomRangeFloat(Minimum=0.2, Maximum=0.3)
    )
    Index 1 Param 3=&FloatFromCurve(FloatCurve=DataInterface() {propertyValues {Curve {keys [
         {interpMode=RCIM_Cubic, leaveTangent=6.3917, tangentMode=RCTM_Break, time=0, value=0.2}
         {arriveTangent=1.7549, interpMode=RCIM_Cubic, leaveTangent=10.6767, tangentMode=RCTM_User, time=0.25, value=1}
        ]}}})
    Index 1 Param 4=&FloatFromCurve(
     FloatCurve=DataInterface() {propertyValues {Curve {keys [{arriveTangent=-0.136, interpMode=RCIM_Cubic, leaveTangent=-0.136, tangentMode=RCTM_User, time=0, value=1}, <at 1 0.95>]}}}
     Scale Curve=&RandomRangeFloat(Minimum=0.5, Maximum=1)
    )
    Index 2 Param 1=&FloatFromCurve(FloatCurve=DataInterface() {propertyValues {Curve {keys [{interpMode=RCIM_Cubic, leaveTangent=1.1368, tangentMode=RCTM_Break, time=0, value=0.45}, <at 1 1>]}}})
    Index 2 Param 2=&FloatFromCurve(
     FloatCurve=DataInterface() {propertyValues {Curve {keys [
         {arriveTangent=1.9056, interpMode=RCIM_Cubic, leaveTangent=1.9056, tangentMode=RCTM_User, time=0, value=0.2}
         {arriveTangent=0.0259, interpMode=RCIM_Cubic, leaveTangent=0.0259, tangentMode=RCTM_User, time=1, value=1.4}
        ]}}}
     Scale Curve=&RandomRangeFloat(Minimum=0.2, Maximum=0.7)
    )
    Index 2 Param 3=&FloatFromCurve(
     FloatCurve=DataInterface() {propertyValues {Curve {keys [
         {interpMode=RCIM_Cubic, leaveTangent=2.3064, tangentMode=RCTM_Break, time=0, value=1}
         {interpMode=RCIM_Cubic, leaveTangent=11.1233, tangentMode=RCTM_Break, time=1, value=2}
        ]}}}
     Scale Curve=1.2
    )
    Index 2 Param 4=&FloatFromCurve(
     FloatCurve=DataInterface() {propertyValues {Curve {keys [
         {arriveTangent=-0.6867, interpMode=RCIM_Cubic, leaveTangent=-0.6867, tangentMode=RCTM_User, time=0, value=1}
         {interpMode=RCIM_Cubic, leaveTangent=-1.2956, tangentMode=RCTM_Break, time=1, value=0.75}
        ]}}}
     Scale Curve=1.2
    )
    Module.Index0Float0Write=true
    Module.Index0Float1Write=true
    Module.Index0Float2Write=true
    Module.Index0Float3Write=true
    Module.Index1Float0Write=true
    Module.Index1Float1Write=true
    Module.Index1Float2Write=true
    Module.Index1Float3Write=true
    Module.Index2Float0Write=true
    Module.Index2Float1Write=true
    Module.Index2Float2Write=true
    Module.Index2Float3Write=true
    Write Parameter Index 1=::NewEnumerator1
    Write Parameter Index 2=::NewEnumerator1
   )
  ]
  rendererClasses [&NiagaraSpriteRendererProperties]
  renderers [{rendererClass=&NiagaraSpriteRendererProperties}]
 }
 Soil {
  emitterUpdateScript [
   &EmitterState(Inactive Response=::NewEnumerator1, Life Cycle Mode=::NewEnumerator1, Loop Behavior=::NewEnumerator1, UseLoopDelay=false)
   &SpawnBurst_Instantaneous(Spawn Count=10, Use Spawn Probability=false, UseLoopCountLimit=false)
   &SpawnBurst_Instantaneous(Spawn Count=10, Spawn Group=1, Spawn Time=1.204, Use Spawn Probability=false, UseLoopCountLimit=false) {module=SpawnBurst_Instantaneous001}
  ]
  particleSpawnScript [
   &InitializeParticle(
    Color Maximum=<rgba 0.2078 0.2078 0.2078 1>
    Color Minimum=<rgba 0.4855 0.4855 0.4855 1>
    Color Mode=::NewEnumerator2
    Lifetime Max=0.4
    Lifetime Min=0.3
    Lifetime Mode=::NewEnumerator1
    Position Offset=&MaskVectorBySpawnGroup(Masked Vector=<0 0 0>, Passthrough Vector=<-8 -10 0>)
    Sprite Size=&RandomRangeVector2D(Minimum=<70 90>, Maximum=<50 70>)
    Sprite Size Mode=::NewEnumerator4
    Sprite UV Mode=::NewEnumerator1
    UsePositionOffset=true
    Write Material Random=true
   )
   &CameraOffset(Camera Offset Amount=10)
   &ShapeLocation(
    Handle Radius=3
    Large Radius=10
    MaskSpawnGroup=false
    Offset Mode=::NewEnumerator0
    Override Seed=false
    Shape Primitive=::NewEnumerator3
    UseOwnerScale=true
   )
   &AddVelocity(
    Module.Use Falloff Distance=false
    Origin Offset=&MaskVectorBySpawnGroup(Masked Vector=<0 0 -1>, Passthrough Vector=<-8 -10 -1>)
    Override Seed=false
    Velocity Mode=::NewEnumerator1
    Velocity Speed=&MaskFloatBySpawnGroup(
     Masked Float=&RandomRangeFloat(Minimum=80, Maximum=30)
     Passthrough Float=&RandomRangeFloat(Minimum=10, Maximum=30)
    )
   ) {module=AddVelocity002}
  ]
  particleUpdateScript [
   &ParticleState()
   &ScaleSpriteSize(
    Non-Uniform Curve Scale=&Multiply_Vector2D_ByFloat(Float=0.8, Vector2D=<0.6 0.5>)
    Non-Uniform Curve Sprite Scale=DataInterface() {propertyValues {
      XCurve {keys [
        {arriveTangent=0.9875, interpMode=RCIM_Cubic, leaveTangent=0.9875, tangentMode=RCTM_User, time=0, value=0.5}
        {arriveTangent=0.5103, interpMode=RCIM_Cubic, leaveTangent=0.5103, tangentMode=RCTM_User, time=0.3, value=0.8}
        {arriveTangent=0.0224, interpMode=RCIM_Cubic, leaveTangent=0.0224, tangentMode=RCTM_User, time=1, value=1}
       ]}
      YCurve {keys [
        {arriveTangent=0.8801, interpMode=RCIM_Cubic, leaveTangent=0.8801, tangentMode=RCTM_User, time=0, value=0.3}
        {arriveTangent=0.6586, interpMode=RCIM_Cubic, leaveTangent=0.6586, tangentMode=RCTM_User, time=0.2, value=0.85}
        {arriveTangent=-0.4778, interpMode=RCIM_Cubic, leaveTangent=-0.4778, tangentMode=RCTM_User, time=1, value=0.8}
       ]}
     }}
    Scale Sprite Size Mode=::NewEnumerator3
   ) {module=ScaleSpriteSize001}
   &ScaleColor(
    Scale Alpha=&FloatFromCurve(
     FloatCurve=DataInterface() {propertyValues {Curve {keys [
         <at 0 0 RCIM_Cubic>
         {arriveTangent=-0.0339, interpMode=RCIM_Cubic, leaveTangent=-0.0339, tangentMode=RCTM_User, time=0.1, value=1}
         {arriveTangent=-1.6214, interpMode=RCIM_Cubic, leaveTangent=-1.6214, tangentMode=RCTM_User, time=0.6616, value=0.2439}
         <at 1 0 RCIM_Cubic>
        ]}}}
     Scale Curve=&Multiply_Float(A=1, B=&FNTimeOfDayScalar(Night Scale=0.6, Morning Scale=0.8, Day Scale=0.8, Evening Scale=0.6))
    )
    Scale RGB=&MakeVectorFromLinearColorRGB(Color=&FNTimeOfDayColor(
      Night Color=&ColorFromCurve(ColorCurve=DataInterface() {propertyValues {
         AlphaCurve {keys []}
         BlueCurve {keys [<at 0.4307 0.28 RCIM_Cubic>,<at 1 0.0222 RCIM_Cubic>]}
         GreenCurve {keys [<at 0.4307 0.3376 RCIM_Cubic>,<at 1 0.0292 RCIM_Cubic>]}
         RedCurve {keys [<at 0.4307 0.3976 RCIM_Cubic>,<at 1 0.04 RCIM_Cubic>]}
        }})
      Morning Color=&ColorFromCurve(ColorCurve=DataInterface() {propertyValues {
         AlphaCurve {keys []}
         BlueCurve {keys [<at 0.4307 0.1094 RCIM_Cubic>,<at 1 0.0148 RCIM_Cubic>]}
         GreenCurve {keys [<at 0.4307 0.3595 RCIM_Cubic>,<at 1 0.0246 RCIM_Cubic>]}
         RedCurve {keys [<at 0.4307 0.6198 RCIM_Cubic>,<at 1 0.04 RCIM_Cubic>]}
        }})
      Day Color=&ColorFromCurve(ColorCurve=DataInterface() {propertyValues {
         AlphaCurve {keys []}
         BlueCurve {keys [<at 0.4307 0.2453 RCIM_Cubic>,<at 1 0.0148 RCIM_Cubic>]}
         GreenCurve {keys [<at 0.4307 0.4288 RCIM_Cubic>,<at 1 0.0246 RCIM_Cubic>]}
         RedCurve {keys [<at 0.4307 0.6198 RCIM_Cubic>,<at 1 0.04 RCIM_Cubic>]}
        }})
      Evening Color=&ColorFromCurve(ColorCurve=DataInterface() {propertyValues {
         AlphaCurve {keys []}
         BlueCurve {keys [<at 0.4307 0.2453 RCIM_Cubic>,<at 1 0.0148 RCIM_Cubic>]}
         GreenCurve {keys [<at 0.4307 0.4288 RCIM_Cubic>,<at 1 0.0246 RCIM_Cubic>]}
         RedCurve {keys [<at 0.4307 0.6198 RCIM_Cubic>,<at 1 0.04 RCIM_Cubic>]}
        }})
     ))
    ScaleA=true
    ScaleRGB=true
   )
   &Drag(Drag=3, Use Linear Drag=true, Use Rotational Drag=true)
   &SolveForcesAndVelocity(Clamp Velocity=false, Limit Acceleration=false)
   &DynamicMaterialParameters(
    Index 0 Param 1=&FloatFromCurve(
     FloatCurve=DataInterface() {propertyValues {Curve {keys [{arriveTangent=-1.447, interpMode=RCIM_Cubic, leaveTangent=-1.447, tangentMode=RCTM_User, time=0, value=0.6}, <at 1 0 RCIM_Cubic>]}}}
     Scale Curve=&RandomRangeFloat(Minimum=0.5, Maximum=1)
    )
    Index 0 Param 2=&RandomRangeFloat(Minimum=0, Maximum=0.5)
    Index 0 Param 3=&RandomRangeFloat(Minimum=0, Maximum=0.5)
    Index 0 Param 4=&RandomRangeFloat(Minimum=1, Maximum=2)
    Module.Index0Float0Write=true
    Module.Index0Float1Write=true
    Module.Index0Float2Write=true
    Module.Index0Float3Write=true
   )
  ]
  rendererClasses [&NiagaraSpriteRendererProperties]
  renderers [{rendererClass=&NiagaraSpriteRendererProperties}]
 }
}
summary {
 systemName=NS_Emote_BeachCavalier_Sand
 emitters [
  {emitterName=Soil, rendererClasses [&NiagaraSpriteRendererProperties]}
  {emitterName=Smoke002, rendererClasses [&NiagaraSpriteRendererProperties]}
  {emitterName=Dirt, rendererClasses [&NiagaraSpriteRendererProperties]}
  {emitterName=BucketPullDust, rendererClasses [&NiagaraSpriteRendererProperties]}
  {emitterName=PollenDust001, rendererClasses [&NiagaraSpriteRendererProperties]}
  {emitterName=HandTapDust, rendererClasses [&NiagaraSpriteRendererProperties]}
 ]
 userVariables [
  User.bIsFrontEnd {type=NiagaraBool}
  User.bIsFrontEndPreview {type=NiagaraBool}
  User.ForwardDot {type=NiagaraFloat, defaultValue=1, description=""}
  User.IsFrontend {type=NiagaraBool}
 ]
}
