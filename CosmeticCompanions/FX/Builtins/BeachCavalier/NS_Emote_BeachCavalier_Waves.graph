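# Niagara effect for the Beach Cavalier emote: a wave that travels along negative Y with spray, leaving wet ground behind.
# Three mesh wake emitters form body, edges and overlay, their positions slid by emitter loop age while dynamic material parameters animate them.
# Splash sprites use spawn groups for differing spray shapes and motion, and a Wetmap emitter lays fading decals.
# Colors and emissive adjust for time of day, performance mode or front end.
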NiagaraSystem /CosmeticCompanions/FX/Builtins/BeachCavalier/NS_Emote_BeachCavalier_Waves
dependencies {
 usedDataInterfaces [
  &NiagaraDataInterfaceCurve
  &NiagaraDataInterfaceColorCurve
  &NiagaraDataInterfaceVectorCurve
  &NiagaraDataInterfacePlatformSet
  &NiagaraDataInterfaceSpriteRendererInfo
 ]
 usedDynamicInputs []
 usedModules [
  &SystemState
  &self:SystemSpawnScript.SystemScriptSource.SystemScriptGraph.NiagaraNodeAssignment_0.SetVariables_#1
 ]
}
emitters {
 Splashes {
  emitterUpdateScript [
   &EmitterState(Life Cycle Mode=::NewEnumerator1, Loop Behavior=::NewEnumerator1, Loop Duration=1.5, UseLoopDelay=false)
   &SpawnRate(
    SpawnRate=&FloatFromCurve(
     FloatCurve=DataInterface() {propertyValues {Curve {keys [<at 0.05 0>,<at 0.1 1>,<at 0.5 1>,<at 0.65 0>]}}}
     CurveIndex=$Emitter.NormalizedLoopAge
     Scale Curve=180
    )
    Use Spawn Probability=false
   )
   &SpawnRate(
    SpawnGroup=1
    SpawnRate=&FloatFromCurve(
     FloatCurve=DataInterface() {propertyValues {Curve {keys [<at 0.1 1>,<at 0.35 1>,<at 0.5 0>]}}}
     CurveIndex=$Emitter.NormalizedLoopAge
     Scale Curve=180
    )
    Use Spawn Probability=false
   ) {module=SpawnRate001}
   &SpawnRate(
    SpawnGroup=2
    SpawnRate=&FloatFromCurve(
     FloatCurve=DataInterface() {propertyValues {Curve {keys [<at 0.125 0>,<at 0.175 1>,<at 0.45 1>,<at 0.75 0>]}}}
     CurveIndex=$Emitter.NormalizedLoopAge
     Scale Curve=256
    )
    Use Spawn Probability=false
   ) {module=SpawnRate002}
  ]
  particleSpawnScript [
   /SetVariables(Particles.VisibilityTag=&SpawnGroup()) {isSetParameters=true}
   &InitializeParticle(
    Color=&MergeHelper_ColorSwitch(
     0=&RandomRangeLinearColor(Minimum=<rgba 0.8698 0.9316 1 1>, Maximum=<rgba 0.6325 0.7666 0.9271 0.8>)
     1=&RandomRangeLinearColor(Minimum=<rgba 4.5827 7.0189 8 0.2>, Maximum=<rgba 6.1159 8 7.6891 0.6>)
     2=&RandomRangeLinearColor(Minimum=<rgba 1.5221 1.6304 1.75 1>, Maximum=<rgba 0.6325 0.7666 0.9271 0.8>)
     3=<rgba 0 0 0 0>
     4=<rgba 0 0 0 0>
     Index=&SpawnGroup()
     Num=::NewEnumerator3
    )
    Lifetime=&RandomRangeFloat(Minimum=0.3, Maximum=0.65)
    Sprite Rotation Mode=::NewEnumerator2
    Sprite Size Mode=::NewEnumerator3
    Uniform Sprite Size=&MergeHelper_FloatSwitch(
     0=&RandomRangeFloat(Minimum=35, Maximum=60)
     1=&RandomRangeFloat(Minimum=25, Maximum=50)
     2=&RandomRangeFloat(Minimum=35, Maximum=60)
     3=0
     Index=&SpawnGroup()
     Num=::NewEnumerator2
    )
    UsePositionOffset=false
    Write Material Random=true
   )
   &ShapeLocation(
    MaskSpawnGroup=false
    Non Uniform Scale=&MergeHelper_VectorSwitch(
     0=<2.25 2.5 1>
     1=<0.4 1.5 1>
     2=&MakeVector(
      X=2.25
      Y=&FloatFromCurve(
       FloatCurve=DataInterface() {propertyValues {Curve {keys [{arriveTangent=1.925, interpMode=RCIM_Cubic, leaveTangent=1.925, tangentMode=RCTM_User, time=0, value=0.2}, <at 1 1 RCIM_Cubic>]}}}
       CurveIndex=$Emitter.NormalizedLoopAge
       Scale Curve=4
      )
      Z=0.2
     )
     3=<0 0 0>
     4=<0 0 0>
     Index=&SpawnGroup()
     Num=::NewEnumerator3
    )
    Offset=&MakeVector(
     X=0
     Y=&Add_Float(
      A=&MergeHelper_FloatSwitch(0=-25, 1=3, 2=-68, 3=0, 4=0, Index=&SpawnGroup(), Num=::NewEnumerator3)
      B=&FloatFromCurve(
       FloatCurve=DataInterface() {propertyValues {Curve {keys [<at 0 0>,<at 1 1>]}}}
       CurveIndex=$Emitter.NormalizedLoopAge
       Scale Curve=&MergeHelper_FloatSwitch(0=-130, 1=-130, 2=-190, 3=0, 4=0, Index=&SpawnGroup(), Num=::NewEnumerator3)
      )
     )
     Z=&MergeHelper_FloatSwitch(0=5, 1=0, 2=5, 3=0, 4=0, Index=&SpawnGroup(), Num=::NewEnumerator3)
    )
    Offset Mode=::NewEnumerator0
    Override Seed=false
    Sphere Radius=24
    Sphere Surface Distribution:NiagaraFloat=&MergeHelper_FloatSwitch(0=0, 1=0, 2=1, 3=0, 4=0, Index=&SpawnGroup(), Num=::NewEnumerator3)
    UseOwnerScale=true
   )
   &SubUVAnimation(
    End Frame Range Override=3
    Evaluation Type=::NewEnumerator0
    "Sprite Renderer "=DataInterface() {propertyValues {SpriteRenderer=&self:Splashes_0.NiagaraSpriteRendererProperties_1}}
    SubUV Animation Mode:ENiagara_SubUVLookupModeV2=::NewEnumerator2
    UseEndFrame=true
    UseStartFrame=true
   )
   &SubUVAnimation(
    End Frame Range Override=3
    Evaluation Type=::NewEnumerator0
    "Sprite Renderer "=DataInterface() {propertyValues {SpriteRenderer=&self:Splashes_0.NiagaraSpriteRendererProperties_8}}
    SubUV Animation Mode:ENiagara_SubUVLookupModeV2=::NewEnumerator2
    UseEndFrame=true
    UseStartFrame=true
   ) {module=SubUVAnimation001}
   &SubUVAnimation(
    End Frame Range Override=3
    Evaluation Type=::NewEnumerator0
    "Sprite Renderer "=DataInterface() {propertyValues {SpriteRenderer=&self:Splashes_0.NiagaraSpriteRendererProperties_2}}
    SubUV Animation Mode:ENiagara_SubUVLookupModeV2=::NewEnumerator2
    UseEndFrame=true
    UseStartFrame=true
   ) {module=SubUVAnimation002}
   &AddVelocity(
    Cone Angle=25
    Cone Axis=<0 -1 0.85>
    Module.UseConeFalloff=false
    Module.UseVelDistribution=false
    Override Seed=false
    Velocity Mode=::NewEnumerator2
    Velocity Speed=&MergeHelper_FloatSwitch(
     0=&RandomRangeFloat(Minimum=280, Maximum=450)
     1=0
     2=&RandomRangeFloat(Minimum=20, Maximum=60)
     3=0
     4=0
     Index=&SpawnGroup()
     Num=::NewEnumerator3
    )
   )
   &AddVelocity(
    Module.Use Falloff Distance=false
    Origin Offset=<0 100 -25>
    Override Seed=false
    Velocity Mode=::NewEnumerator1
    Velocity Origin=$Particles.ShapeLocation.ShapePosition
    Velocity Speed=&MergeHelper_FloatSwitch(
     0=0
     1=&RandomRangeFloat(Minimum=80, Maximum=120)
     2=0
     3=0
     4=0
     Index=&SpawnGroup()
     Num=::NewEnumerator3
    )
   ) {module=AddVelocity001}
  ]
  particleUpdateScript [
   &ParticleState()
   &SpriteRotationRate(Rotation Rate=&Lerp_Float(A=-150, B=150, Alpha=$Particles.Initial.MaterialRandom))
   &ScaleSpriteSize(Uniform Curve Sprite Scale=DataInterface() {propertyValues {Curve {
       defaultValue=0
       keys [
        {arriveTangent=4.1785, interpMode=RCIM_Cubic, leaveTangent=4.1785, tangentMode=RCTM_User, time=0, value=0}
        {arriveTangent=1.1179, interpMode=RCIM_Cubic, leaveTangent=1.1179, tangentMode=RCTM_User, time=0.2042, value=0.6228}
        {arriveTangent=0.0224, interpMode=RCIM_Cubic, leaveTangent=0.0224, tangentMode=RCTM_User, time=1, value=1}
       ]
      }}})
   /SetVariables(
    Particles.0DP:Vector4f=&MakeVector4(
     X=&FloatFromCurve(FloatCurve=DataInterface() {propertyValues {Curve {keys [
          {arriveTangent=-2.7755, interpMode=RCIM_Cubic, leaveTangent=-2.7755, tangentMode=RCTM_User, time=0.25, value=1}
          {arriveTangent=-0.7689, interpMode=RCIM_Cubic, leaveTangent=-0.7689, tangentMode=RCTM_User, time=1.0011, value=0}
         ]}}})
     Y=&FloatFromCurve(FloatCurve=DataInterface() {propertyValues {Curve {keys [<at -0.0086 -1.3001>,<at 0.9945 1.9153>]}}})
     Z=0
     W=0
    )
    Particles.1DP:Vector4f=&MakeVector4(
     X=&FloatFromCurve(FloatCurve=DataInterface() {propertyValues {Curve {keys [
          {interpMode=RCIM_Cubic, tangentMode=RCTM_User, time=0, value=1}
          {arriveTangent=-3.7889, interpMode=RCIM_Cubic, leaveTangent=-3.7889, tangentMode=RCTM_User, time=0.8, value=0}
         ]}}})
     Y=0
     Z=0
     W=0
    )
   ) {isSetParameters=true}
   &GravityForce(Gravity=&MakeVector(
     X=0
     Y=0
     Z=&MergeHelper_FloatSwitch(
      0=&RandomRangeFloat(Minimum=-980, Maximum=-1200)
      1=&RandomRangeFloat(Minimum=-120, Maximum=-300)
      2=&RandomRangeFloat(Minimum=-120, Maximum=-300)
      3=0
      4=0
      Index=&SpawnGroup()
      Num=::NewEnumerator3
     )
    ))
   &SolveForcesAndVelocity(Clamp Velocity=false, Limit Acceleration=false)
   &ScaleSpriteSizeBySpeed(
    Max Scale Factor=&MergeHelper_Vector2Switch(0=<1 1>, 1=<1 3>, 2=<1 1>, 3=<0 0>, 4=<0 0>, Index=&SpawnGroup(), Num=::NewEnumerator3)
    Min Scale Factor=&MergeHelper_Vector2Switch(0=<1 1>, 1=<1 1>, 2=<1 1>, 3=<0 0>, 4=<0 0>, Index=&SpawnGroup(), Num=::NewEnumerator3)
    Velocity Threshold=250
   )
   &ScaleColor(
    Scale Alpha=4
    Scale RGB=&MakeVectorFromLinearColorRGB(Color=&RandomRangeLinearColor(
      Minimum=&FNTimeOfDayColor(Night Color=<rgba 1.5 1.5 1.5 1>, Morning Color=<rgba 4 4 4 1>, Day Color=<rgba 4 4 4 1>, Evening Color=<rgba 4 4 4 1>)
      Maximum=&FNTimeOfDayColor(Night Color=<rgba 0.8 0.8 0.8 1>, Morning Color=<rgba 2 2 2 1>, Day Color=<rgba 2 2 2 1>, Evening Color=<rgba 2 2 2 1>)
     ))
    ScaleA=true
    ScaleRGB=true
   )
  ]
  rendererClasses [&NiagaraSpriteRendererProperties]
  renderers [
   {rendererClass=&NiagaraSpriteRendererProperties}
   {rendererClass=&NiagaraSpriteRendererProperties, rendererIndex=1}
   {rendererClass=&NiagaraSpriteRendererProperties, rendererIndex=2}
  ]
 }
 Wakes_Moving {
  emitterUpdateScript [
   &EmitterState(Inactive Response=::NewEnumerator2, Life Cycle Mode=::NewEnumerator1, Loop Behavior=::NewEnumerator1, Loop Duration=1.4, UseLoopDelay=false)
   &SpawnRate(
    SpawnRate=&FloatFromCurve(
     FloatCurve=DataInterface() {propertyValues {Curve {keys [<at 0 1>,<at 0.475 1>,<at 0.65 0>]}}}
     CurveIndex=$Emitter.NormalizedLoopAge
     Scale Curve=32
    )
    Use Spawn Probability=false
   )
   &SpawnRate(
    SpawnGroup=1
    SpawnRate=&FloatFromCurve(
     FloatCurve=DataInterface() {propertyValues {Curve {keys [<at 0.125 0>,<at 0.3 1>,<at 0.6 1>,<at 0.65 0>]}}}
     CurveIndex=$Emitter.NormalizedLoopAge
     Scale Curve=18
    )
    Use Spawn Probability=false
   ) {module=SpawnRate001}
   &SpawnRate(
    SpawnGroup=2
    SpawnRate=&FloatFromCurve(
     FloatCurve=DataInterface() {propertyValues {Curve {keys [<at 0.125 0>,<at 0.3 1>,<at 0.6 1>,<at 0.65 0>]}}}
     CurveIndex=$Emitter.NormalizedLoopAge
     Scale Curve=18
    )
    Use Spawn Probability=false
   ) {module=SpawnRate002}
   /SetVariables(
    Emitter.OverallEmissiveAdd:NiagaraFloat=&FN_FrontEndFloatModifier(Gameplay=0, Front End=0.3)
    Emitter.ReflectionIntensity:NiagaraFloat=&FN_FrontEndFloatModifier(Gameplay=2, Front End=2)
    Emitter.WaterTint:LinearColor=&ColorFromCurve(
     ColorCurve=DataInterface() {propertyValues {
       AlphaCurve {keys [<at 0 1>]}
       BlueCurve {keys [<at 0.305 1 RCIM_Cubic>,<at 0.6508 0.7969>,<at 1 0.3646>]}
       GreenCurve {keys [{arriveTangent=-0.5438, interpMode=RCIM_Cubic, time=0.305, value=1}, <at 0.6508 0.5007>, <at 1 0.2166>]}
       RedCurve {keys [{arriveTangent=-0.9068, interpMode=RCIM_Cubic, time=0.305, value=1}, <at 0.6508 0.303>, <at 1 0.1177>]}
      }}
     CurveIndex=$Emitter.NormalizedLoopAge
    )
   ) {isSetParameters=true, module=SetVariables_#2}
  ]
  particleSpawnScript [
   /SetVariables(Particles.VisibilityTag=&SpawnGroup()) {isSetParameters=true, module=SetVariables_#3}
   &InitializeParticle(
    Color=&FNTimeOfDayColor(
     Night Color=<rgba 0.3278 0.4969 0.6172 1>
     Morning Color=<rgba 0.5271 0.7991 1 1>
     Day Color=<rgba 0.5271 0.7991 1 1>
     Evening Color=<rgba 0.5271 0.7991 1 1>
    )
    Lifetime=0.7
    Mesh Scale=&Multiply_VectorByFloat(Vector=&UniformRangedVector(Minimum=<1.2 1.6 0.8>, Maximum=<1.3 2 1>), Float=0.45)
    Mesh Scale Mode=::NewEnumerator4
    Position Mode=::NewEnumerator0
    Sprite Size Mode=::NewEnumerator3
    Write Material Random=true
   ) {module=InitializeParticle001}
   &ShapeLocation(
    Box / Plane Mode:ENiagara_BoxPlaneMode=::NewEnumerator1
    MaskSpawnGroup=false
    Override Seed=false
    Plane Size:Vector2f=<45 10>
    Shape Primitive=::NewEnumerator2
    UseOwnerScale=true
   )
   &AddVelocity(Velocity=<0 0 5>)
   &InitialMeshOrientation(Mesh Orientation Mode=::NewEnumerator2, Orientation Vector=<0 1 0>, Use Rotation Vector=true)
  ]
  particleUpdateScript [
   &ParticleState()
   &ScaleMeshSize(Scale Factor=&Multiply_Vector(
     A=&VectorFromCurve(VectorCurve=DataInterface() {propertyValues {
        XCurve {keys [
          {arriveTangent=4.6994, interpMode=RCIM_Cubic, leaveTangent=4.6994, tangentMode=RCTM_User, time=0, value=0.2795}
          {arriveTangent=0.4589, interpMode=RCIM_Cubic, leaveTangent=0.4589, tangentMode=RCTM_User, time=0.15, value=0.7609}
          <at 1 1>
         ]}
        YCurve {keys [
          {arriveTangent=1.7936, interpMode=RCIM_Cubic, leaveTangent=1.7936, tangentMode=RCTM_User, time=0, value=0.2}
          <at 0.4 0.4>
          {arriveTangent=-0.1215, interpMode=RCIM_Cubic, leaveTangent=-0.1215, tangentMode=RCTM_User, time=1, value=1}
         ]}
        ZCurve {keys [<at 0 0.2 RCIM_Cubic>, {arriveTangent=-0.2002, interpMode=RCIM_Cubic, leaveTangent=-0.2002, time=0.33, value=1}, <at 0.9988 0>]}
       }})
     B=&MakeVector(
      X=1
      Y=1
      Z=&FloatFromCurve(
       FloatCurve=DataInterface() {propertyValues {Curve {keys [
           <at 0 1 RCIM_Cubic>
           <at 0.25 1.1 RCIM_Cubic>
           {arriveTangent=-1.4667, interpMode=RCIM_Cubic, leaveTangent=-1.4667, time=0.5, value=1}
           <at 1 0 RCIM_Cubic>
          ]}}}
       CurveIndex=$Emitter.NormalizedLoopAge
      )
     )
    ))
   &Drag(Drag=0, Rotational Drag=20, Use Linear Drag=true, Use Rotational Drag=true)
   &ScaleColor(
    Scale Alpha=&FloatFromCurve(
     FloatCurve=DataInterface() {propertyValues {Curve {keys [
         {arriveTangent=9.9926, interpMode=RCIM_Cubic, leaveTangent=9.9927, tangentMode=RCTM_User, time=0, value=0}
         {arriveTangent=-0.1957, interpMode=RCIM_Cubic, leaveTangent=-0.1957, tangentMode=RCTM_User, time=0.2, value=1}
         <at 1.0011 0.002 RCIM_Cubic>
        ]}}}
     Scale Curve=0.8
    )
    Scale RGB=&VectorFromFloat(Value=1)
    ScaleA=true
    ScaleRGB=true
   )
   &SolveForcesAndVelocity(Clamp Velocity=false, Limit Acceleration=false)
   &DynamicMaterialParameters(
    Index 0 Param 1=&FloatFromCurve(FloatCurve=DataInterface() {propertyValues {Curve {keys [
         {arriveTangent=0.2615, interpMode=RCIM_Cubic, leaveTangent=0.2615, tangentMode=RCTM_User, time=0, value=-1.7984}
         {arriveTangent=3.2109, interpMode=RCIM_Cubic, leaveTangent=3.2109, tangentMode=RCTM_User, time=0.9985, value=0}
        ]}}})
    Index 0 Param 2=&Add_Float(
     A=&FloatFromCurve(FloatCurve=DataInterface() {propertyValues {Curve {keys [
          {arriveTangent=-9.4845, interpMode=RCIM_Cubic, leaveTangent=-9.4845, tangentMode=RCTM_User, time=0, value=4}
          {arriveTangent=0.2383, interpMode=RCIM_Cubic, leaveTangent=0.2383, tangentMode=RCTM_User, time=1, value=0}
         ]}}})
     B=0
    )
    Index 1 Param 1=&FloatFromCurve(FloatCurve=DataInterface() {propertyValues {Curve {keys [
         <at 0 0.1 RCIM_Cubic>
         {arriveTangent=0.9, interpMode=RCIM_Cubic, leaveTangent=0.9, time=0.5861, value=0.2}
         {arriveTangent=3.3294, interpMode=RCIM_Cubic, leaveTangent=3.3294, tangentMode=RCTM_User, time=1, value=1}
        ]}}})
    Module.Index0Float0Write=true
    Module.Index0Float1Write=true
    Module.Index0Float2Write=true
    Module.Index0Float3Write=true
    Module.Index1Float0Write=true
    Module.Index1Float1Write=true
    Module.Index1Float2Write=true
    Module.Index1Float3Write=true
    Write Parameter Index 1=::NewEnumerator1
   )
   /SetVariables(Particles.Position=&AddVectorToPosition(
     Position=&MakeVector(
      X=0
      Y=&FloatFromCurve(FloatCurve=DataInterface() {propertyValues {Curve {keys [<at 0 0>,<at 1 2>]}}}, CurveIndex=$Emitter.NormalizedLoopAge, Scale Curve=-100)
      Z=0
     )
     Vector=&AddVectorToPosition(
      Position=$Particles.Initial.ShapeLocation.ShapePosition
      Vector=&Multiply_VectorByFloat(Vector=$Particles.Velocity, Float=$Engine.DeltaTime)
     )
    )) {isSetParameters=true, module=SetVariables_#4}
   /SetVariables(Particles.Position2:NiagaraPosition=&AddVectorToPosition(
     Position=&MakeVector(
      X=0
      Y=&FloatFromCurve(FloatCurve=DataInterface() {propertyValues {Curve {keys [<at 0 0>,<at 1 2>]}}}, CurveIndex=$Emitter.NormalizedLoopAge, Scale Curve=-75)
      Z=0
     )
     Vector=&AddVectorToPosition(
      Position=$Particles.Initial.ShapeLocation.ShapePosition
      Vector=&Multiply_VectorByFloat(Vector=$Particles.Velocity, Float=$Engine.DeltaTime)
     )
    )) {isSetParameters=true}
   /SetVariables(Particles.Position3:NiagaraPosition=&AddVectorToPosition(
     Position=&MakeVector(
      X=0
      Y=&FloatFromCurve(FloatCurve=DataInterface() {propertyValues {Curve {keys [<at 0 0>,<at 1 2>]}}}, CurveIndex=$Emitter.NormalizedLoopAge, Scale Curve=-125)
      Z=0
     )
     Vector=&AddVectorToPosition(
      Position=$Particles.Initial.ShapeLocation.ShapePosition
      Vector=&Multiply_VectorByFloat(Vector=$Particles.Velocity, Float=$Engine.DeltaTime)
     )
    )) {isSetParameters=true}
  ]
  rendererClasses [&NiagaraMeshRendererProperties]
  renderers [
   {rendererClass=&NiagaraMeshRendererProperties, rendererIndex=0}
   {rendererClass=&NiagaraMeshRendererProperties, rendererIndex=1}
   {rendererClass=&NiagaraMeshRendererProperties, rendererIndex=2}
  ]
 }
 Wakes_Moving_Edge {
  emitterSpawnScript [/SetVariables(Emitter.OverallEmissiveAdd:NiagaraFloat=&FN_FrontEndFloatModifier(Gameplay=0, Front End=0.5)) {isSetParameters=true}]
  emitterUpdateScript [
   &EmitterState(Life Cycle Mode=::NewEnumerator1, Loop Behavior=::NewEnumerator1, Loop Duration=1.2, UseLoopDelay=false)
   &SpawnRate(
    SpawnRate=&FloatFromCurve(
     FloatCurve=DataInterface() {propertyValues {Curve {keys [<at 0 1>,<at 0.325 1>,<at 0.38 0>]}}}
     CurveIndex=$Emitter.NormalizedLoopAge
     Scale Curve=120
    )
    Use Spawn Probability=false
   )
  ]
  particleSpawnScript [
   &FN_SetSpawnGroup(Number of Spawn Groups=2)
   /SetVariables(Particles.FrontToBack:NiagaraFloat=&RandomRangeFloat(Minimum=0, Maximum=1)) {isSetParameters=true}
   &InitializeParticle(
    Color=&MakeCustomLinearColorFromBool(
     Bool=&IsPlatformSetActive(New Platform Set=DataInterface() {propertyValues {Platforms {
         qualityLevelMask=-31
         cVarConditions [{
           bUseMaxFloat=false
           bUseMaxInt=false
           bUseMinFloat=true
           bUseMinInt=true
           cVarName=athena.isperfmode
           failResponse=Disable
           maxFloat=1
           maxInt=1
           minFloat=1
           minInt=1
           passResponse=None
           value=true
          }]
        }}})
     True Color=<rgba 0.4282 1 0.9351 1>
     False Color=<rgba 1 1 1 1>
    )
    Lifetime Max=1
    Lifetime Min=0.75
    Lifetime Mode=::NewEnumerator1
    Mesh Scale=&Multiply_Vector(
     A=&Multiply_VectorByFloat(Vector=&RandomRangeVector(Minimum=<0.75 1 0.75>, Maximum=<1 1.4 1.1>), Float=0.7)
     B=&VectorFromCurve(
      VectorCurve=DataInterface() {propertyValues {XCurve {keys [<at 0 0.5291>,<at 1 1>]}, YCurve {keys [<at 0 1>,<at 1 1>]}, ZCurve {keys [<at 0 1>,<at 1 1>]}}}
      CurveIndex=$Particles.FrontToBack
     )
    )
    Mesh Scale Mode=::NewEnumerator4
    Sprite Rotation Mode=::NewEnumerator2
    UsePositionOffset=false
    Write Material Random=true
   ) {module=InitializeParticle001}
   /SetVariables(Particles.Scale=&Multiply_Vector(
     A=&VectorFromCurve(
      VectorCurve=DataInterface() {propertyValues {
        XCurve {keys [{arriveTangent=1.7846, interpMode=RCIM_Cubic, leaveTangent=1.7846, tangentMode=RCTM_User, time=0, value=0.4}, <at 1 1 RCIM_Cubic>]}
        YCurve {keys [{arriveTangent=1.7846, interpMode=RCIM_Cubic, leaveTangent=1.7846, tangentMode=RCTM_User, time=0, value=0.4}, <at 1 1 RCIM_Cubic>]}
        ZCurve {keys [{arriveTangent=1.7846, interpMode=RCIM_Cubic, leaveTangent=1.7846, tangentMode=RCTM_User, time=0, value=0.4}, <at 1 1 RCIM_Cubic>]}
       }}
      CurveIndex=$Particles.FrontToBack
      Scale Curve=&Multiply_VectorByFloat(Vector=$Particles.Scale, Float=&Lerp_Float(A=0.8, B=1.4, Alpha=0.5))
     )
     B=&Lerp_Vector(A=<0.6 0.6 0.6>, B=<0.8 0.8 0.8>, Alpha=$Particles.LeanLerp)
    )) {isSetParameters=true}
   &ShapeLocation(
    Box Midpoint=<0.5 0.5 0>
    Box Size=<15 80 20>
    MaskSpawnGroup=false
    Offset=&MergeHelper_VectorSwitch(0=<-22.5 10 0>, 1=<22.5 10 0>, 2=<0 0 0>, 3=?, Index=&SpawnGroup(), Num=::NewEnumerator1)
    Offset Mode=::NewEnumerator0
    Override Seed=false
    Shape Primitive=::NewEnumerator2
    Surface Only=false
    UseOwnerScale=true
   )
   &InitialMeshOrientation(
    Mesh Orientation Mode=::NewEnumerator1
    Rotation=&MergeHelper_VectorSwitch(
     0=&RandomRangeVector(Minimum=<0 0 -0.0125>, Maximum=<0 0 -0.08>)
     1=&RandomRangeVector(Minimum=<0 0 0.0125>, Maximum=<0 0 0.08>)
     2=<0 0 0>
     3=?
     Index=&SpawnGroup()
     Num=::NewEnumerator1
    )
    Use Rotation Vector=true
   )
  ]
  particleUpdateScript [
   &ParticleState()
   /SetVariables(Particles.Scale=&Multiply_Vector(
     A=$Particles.Initial.Scale
     B=&VectorFromCurve(VectorCurve=DataInterface() {propertyValues {
        XCurve {keys [
          {arriveTangent=4.6994, interpMode=RCIM_Cubic, leaveTangent=4.6994, tangentMode=RCTM_User, time=0, value=0.0795}
          {arriveTangent=0.4589, interpMode=RCIM_Cubic, leaveTangent=0.4589, tangentMode=RCTM_User, time=0.1447, value=0.7609}
          <at 1 1>
         ]}
        YCurve {keys [{arriveTangent=2.5596, interpMode=RCIM_Cubic, leaveTangent=2.5596, tangentMode=RCTM_User, time=0, value=0.0767}, <at 1 1>]}
        ZCurve {keys [
          <at 0 0 RCIM_Cubic>
          {arriveTangent=-0.1527, interpMode=RCIM_Cubic, leaveTangent=-0.1527, tangentMode=RCTM_User, time=0.336, value=1}
          <at 0.9988 0>
         ]}
       }})
    )) {isSetParameters=true}
   &SolveForcesAndVelocity(Clamp Velocity=false, Limit Acceleration=false)
   &DynamicMaterialParameters(
    Index 0 Param 1=&FloatFromCurve(FloatCurve=DataInterface() {propertyValues {Curve {keys [
         {arriveTangent=1.2629, interpMode=RCIM_Cubic, leaveTangent=1.2629, tangentMode=RCTM_User, time=-0.0045, value=-1.9799}
         {arriveTangent=4.4591, interpMode=RCIM_Cubic, leaveTangent=4.4591, tangentMode=RCTM_User, time=1, value=0}
        ]}}})
    Index 0 Param 2=&FloatFromCurve(FloatCurve=DataInterface() {propertyValues {Curve {keys [<at 0.0102 4>,<at 0.9749 0>]}}})
    Index 0 Param 3=&MakeFloatFromVector(Vector=$Particles.Scale, Channel=::NewEnumerator0)
    Index 0 Param 4=1
    Index 1 Param 1=&FloatFromCurve(FloatCurve=DataInterface() {propertyValues {Curve {keys [<at 0.7 0 RCIM_Cubic>, {arriveTangent=6.4641, interpMode=RCIM_Cubic, leaveTangent=6.4641, tangentMode=RCTM_User, time=1, value=1}]}}})
    Module.Index0Float0Write=true
    Module.Index0Float1Write=true
    Module.Index0Float2Write=true
    Module.Index0Float3Write=true
    Module.Index1Float0Write=true
    Module.Index1Float1Write=true
    Module.Index1Float2Write=true
    Module.Index1Float3Write=true
    Write Parameter Index 1=::NewEnumerator1
   )
   &ScaleColor(
    Scale Alpha=&FloatFromCurve(FloatCurve=DataInterface() {propertyValues {Curve {keys [{interpMode=RCIM_Cubic, tangentMode=RCTM_User, time=0, value=0}, {interpMode=RCIM_Cubic, tangentMode=RCTM_User, time=0.16, value=1}]}}})
    ScaleA=true
    ScaleRGB=true
   )
   /SetVariables(Particles.Position=&AddVectorToPosition(
     Position=&MakeVector(
      X=0
      Y=&FloatFromCurve(FloatCurve=DataInterface() {propertyValues {Curve {keys [<at 0 0>,<at 1 2>]}}}, CurveIndex=$Emitter.NormalizedLoopAge, Scale Curve=-100)
      Z=0
     )
     Vector=&AddVectorToPosition(
      Position=$Particles.Initial.ShapeLocation.ShapePosition
      Vector=&Multiply_VectorByFloat(Vector=$Particles.Velocity, Float=$Engine.DeltaTime)
     )
    )) {isSetParameters=true}
   /SetVariables(Particles.Position2:NiagaraPosition=&AddVectorToPosition(
     Position=&MakeVector(
      X=0
      Y=&FloatFromCurve(FloatCurve=DataInterface() {propertyValues {Curve {keys [<at 0 0>,<at 1 2>]}}}, CurveIndex=$Emitter.NormalizedLoopAge, Scale Curve=-75)
      Z=0
     )
     Vector=&AddVectorToPosition(
      Position=$Particles.Initial.ShapeLocation.ShapePosition
      Vector=&Multiply_VectorByFloat(Vector=$Particles.Velocity, Float=$Engine.DeltaTime)
     )
    )) {isSetParameters=true}
  ]
  rendererClasses [&NiagaraMeshRendererProperties]
  renderers [
   {rendererClass=&NiagaraMeshRendererProperties, rendererIndex=0}
   {rendererClass=&NiagaraMeshRendererProperties, rendererIndex=1}
  ]
 }
 Wakes_Moving_Overlay {
  emitterUpdateScript [
   &EmitterState(Inactive Response=::NewEnumerator2, Life Cycle Mode=::NewEnumerator1, Loop Behavior=::NewEnumerator1, Loop Duration=1.4, UseLoopDelay=false)
   &SpawnBurst_Instantaneous(Spawn Count=4, Use Spawn Probability=false, UseLoopCountLimit=false) {enabled=false}
   &SpawnRate(
    SpawnRate=&FloatFromCurve(
     FloatCurve=DataInterface() {propertyValues {Curve {keys [<at 0 1>,<at 0.325 1>,<at 0.55 0>]}}}
     CurveIndex=$Emitter.NormalizedLoopAge
     Scale Curve=24
    )
    Use Spawn Probability=false
   )
   /SetVariables(
    Emitter.OverallEmissiveAdd:NiagaraFloat=&FN_FrontEndFloatModifier(Gameplay=1, Front End=0.3)
    Emitter.ReflectionIntensity:NiagaraFloat=&FN_FrontEndFloatModifier(Gameplay=2, Front End=2)
    Emitter.WaterTint:LinearColor=&ColorFromCurve(
     ColorCurve=DataInterface() {propertyValues {
       AlphaCurve {keys [<at 0 1>]}
       BlueCurve {keys [<at 0.305 1 RCIM_Cubic>,<at 0.6508 0.7969>,<at 1 0.3646>]}
       GreenCurve {keys [{arriveTangent=-0.5438, interpMode=RCIM_Cubic, time=0.305, value=1}, <at 0.6508 0.5007>, <at 1 0.2166>]}
       RedCurve {keys [{arriveTangent=-0.9068, interpMode=RCIM_Cubic, time=0.305, value=1}, <at 0.6508 0.303>, <at 1 0.1177>]}
      }}
     CurveIndex=$Emitter.NormalizedLoopAge
    )
   ) {isSetParameters=true, module=SetVariables_#2}
  ]
  particleSpawnScript [
   /SetVariables(Particles.VisibilityTag=&SpawnGroup()) {isSetParameters=true, module=SetVariables_#3}
   &InitializeParticle(
    Color=&FNTimeOfDayColor(Night Color=<rgba 0.4233 0.4233 0.4233 1>, Morning Color=<rgba 1 1 1 1>, Day Color=<rgba 1 1 1 1>, Evening Color=<rgba 1 1 1 1>)
    Lifetime=0.7
    Mesh Scale=&Multiply_VectorByFloat(Vector=&UniformRangedVector(Minimum=<1.2 1 1>, Maximum=<1.4 1.2 1.15>), Float=0.6)
    Mesh Scale Mode=::NewEnumerator4
    Position Mode=::NewEnumerator0
    Sprite Size Mode=::NewEnumerator3
    Write Material Random=true
   ) {module=InitializeParticle001}
   &ShapeLocation(
    Box / Plane Mode:ENiagara_BoxPlaneMode=::NewEnumerator1
    MaskSpawnGroup=false
    Override Seed=false
    Plane Size:Vector2f=<65 10>
    Shape Primitive=::NewEnumerator2
    UseOwnerScale=true
   )
   &AddVelocity(Velocity=<0 0 5>)
   &InitialMeshOrientation(Mesh Orientation Mode=::NewEnumerator2, Orientation Vector=<0 1 0>, Use Rotation Vector=true)
  ]
  particleUpdateScript [
   &ParticleState()
   &ScaleMeshSize(Scale Factor=&Multiply_Vector(
     A=&VectorFromCurve(VectorCurve=DataInterface() {propertyValues {
        XCurve {keys [
          {arriveTangent=4.6994, interpMode=RCIM_Cubic, leaveTangent=4.6994, tangentMode=RCTM_User, time=0, value=0.2795}
          {arriveTangent=0.4589, interpMode=RCIM_Cubic, leaveTangent=0.4589, tangentMode=RCTM_User, time=0.15, value=0.7609}
          <at 1 1>
         ]}
        YCurve {keys [
          {arriveTangent=1.7936, interpMode=RCIM_Cubic, leaveTangent=1.7936, tangentMode=RCTM_User, time=0, value=0.2}
          <at 0.4 0.4>
          {arriveTangent=-0.1215, interpMode=RCIM_Cubic, leaveTangent=-0.1215, tangentMode=RCTM_User, time=1, value=1}
         ]}
        ZCurve {keys [<at 0 0.2 RCIM_Cubic>, {arriveTangent=-0.2002, interpMode=RCIM_Cubic, leaveTangent=-0.2002, time=0.33, value=1}, <at 0.9988 0>]}
       }})
     B=&MakeVector(
      X=1
      Y=1
      Z=&FloatFromCurve(
       FloatCurve=DataInterface() {propertyValues {Curve {keys [
           <at 0 1 RCIM_Cubic>
           <at 0.25 1.1 RCIM_Cubic>
           {arriveTangent=-1.4667, interpMode=RCIM_Cubic, leaveTangent=-1.4667, time=0.5, value=1}
           <at 1 0 RCIM_Cubic>
          ]}}}
       CurveIndex=$Emitter.NormalizedLoopAge
      )
     )
    ))
   &Drag(Drag=0, Rotational Drag=20, Use Linear Drag=true, Use Rotational Drag=true)
   &ScaleColor(
    Scale Alpha=&FloatFromCurve(FloatCurve=DataInterface() {propertyValues {Curve {keys [
         {arriveTangent=9.9926, interpMode=RCIM_Cubic, leaveTangent=9.9927, tangentMode=RCTM_User, time=0, value=0}
         {arriveTangent=-0.1957, interpMode=RCIM_Cubic, leaveTangent=-0.1957, tangentMode=RCTM_User, time=0.2, value=1}
         <at 1.0011 0.002 RCIM_Cubic>
        ]}}})
    Scale RGB=&VectorFromFloat(Value=1)
    ScaleA=true
    ScaleRGB=true
   )
   &SolveForcesAndVelocity(Clamp Velocity=false, Limit Acceleration=false)
   &DynamicMaterialParameters(
    Index 0 Param 1=&FloatFromCurve(FloatCurve=DataInterface() {propertyValues {Curve {keys [
         {arriveTangent=0.2615, interpMode=RCIM_Cubic, leaveTangent=0.2615, tangentMode=RCTM_User, time=0, value=-1.7984}
         {arriveTangent=3.2109, interpMode=RCIM_Cubic, leaveTangent=3.2109, tangentMode=RCTM_User, time=0.9985, value=0}
        ]}}})
    Index 0 Param 2=&Add_Float(
     A=&FloatFromCurve(FloatCurve=DataInterface() {propertyValues {Curve {keys [
          {arriveTangent=-9.4845, interpMode=RCIM_Cubic, leaveTangent=-9.4845, tangentMode=RCTM_User, time=0, value=4}
          {arriveTangent=0.2383, interpMode=RCIM_Cubic, leaveTangent=0.2383, tangentMode=RCTM_User, time=1, value=0}
         ]}}})
     B=0
    )
    Index 1 Param 1=&FloatFromCurve(FloatCurve=DataInterface() {propertyValues {Curve {keys [
         <at 0 0.1 RCIM_Cubic>
         {arriveTangent=0.9, interpMode=RCIM_Cubic, leaveTangent=0.9, time=0.5861, value=0.2}
         {arriveTangent=3.3294, interpMode=RCIM_Cubic, leaveTangent=3.3294, tangentMode=RCTM_User, time=1, value=1}
        ]}}})
    Module.Index0Float0Write=true
    Module.Index0Float1Write=true
    Module.Index0Float2Write=true
    Module.Index0Float3Write=true
    Module.Index1Float0Write=true
    Module.Index1Float1Write=true
    Module.Index1Float2Write=true
    Module.Index1Float3Write=true
    Write Parameter Index 1=::NewEnumerator1
   )
   /SetVariables(Particles.Position=&AddVectorToPosition(
     Position=&MakeVector(
      X=0
      Y=&FloatFromCurve(FloatCurve=DataInterface() {propertyValues {Curve {keys [<at 0 0>,<at 1 2>]}}}, CurveIndex=$Emitter.NormalizedLoopAge, Scale Curve=-80)
      Z=0
     )
     Vector=&AddVectorToPosition(
      Position=$Particles.Initial.ShapeLocation.ShapePosition
      Vector=&Multiply_VectorByFloat(Vector=$Particles.Velocity, Float=$Engine.DeltaTime)
     )
    )) {isSetParameters=true, module=SetVariables_#4}
  ]
  rendererClasses [&NiagaraMeshRendererProperties]
  renderers [{rendererClass=&NiagaraMeshRendererProperties, rendererIndex=0}]
 }
 Wetmap {
  emitterUpdateScript [
   &EmitterState(Life Cycle Mode=::NewEnumerator1, Loop Behavior=::NewEnumerator1, Loop Duration=3.6, UseLoopDelay=false)
   &SpawnRate(
    SpawnRate=&FloatFromCurve(
     FloatCurve=DataInterface() {propertyValues {Curve {keys [
         <at 0.1 0 RCIM_Cubic>
         {arriveTangent=-0.1741, interpMode=RCIM_Cubic, leaveTangent=-0.1741, tangentMode=RCTM_User, time=0.125, value=1}
         {arriveTangent=0.018, interpMode=RCIM_Cubic, leaveTangent=0.018, tangentMode=RCTM_User, time=0.225, value=1}
         {arriveTangent=-4.7075, interpMode=RCIM_Cubic, leaveTangent=-4.7075, tangentMode=RCTM_User, time=0.325, value=0}
        ]}}}
     CurveIndex=$Emitter.NormalizedLoopAge
     Scale Curve=128
    )
    Use Spawn Probability=false
   )
   &SpawnRate(
    SpawnGroup=2
    SpawnRate=&FloatFromCurve(
     FloatCurve=DataInterface() {propertyValues {Curve {keys [
         <at 0.1 0 RCIM_Cubic>
         {arriveTangent=-0.1741, interpMode=RCIM_Cubic, leaveTangent=-0.1741, tangentMode=RCTM_User, time=0.125, value=1}
         {arriveTangent=0.018, interpMode=RCIM_Cubic, leaveTangent=0.018, tangentMode=RCTM_User, time=0.225, value=1}
         {arriveTangent=-4.7075, interpMode=RCIM_Cubic, leaveTangent=-4.7075, tangentMode=RCTM_User, time=0.325, value=0}
        ]}}}
     CurveIndex=$Emitter.NormalizedLoopAge
     Scale Curve=64
    )
    Use Spawn Probability=false
   ) {module=SpawnRate001}
   &SpawnRate(
    SpawnGroup=1
    SpawnRate=&FloatFromCurve(
     FloatCurve=DataInterface() {propertyValues {Curve {keys [
         <at 0.1 0 RCIM_Cubic>
         {arriveTangent=-0.1741, interpMode=RCIM_Cubic, leaveTangent=-0.1741, tangentMode=RCTM_User, time=0.125, value=1}
         {arriveTangent=0.018, interpMode=RCIM_Cubic, leaveTangent=0.018, tangentMode=RCTM_User, time=0.225, value=1}
         {arriveTangent=-4.7075, interpMode=RCIM_Cubic, leaveTangent=-4.7075, tangentMode=RCTM_User, time=0.325, value=0}
        ]}}}
     CurveIndex=$Emitter.NormalizedLoopAge
     Scale Curve=64
    )
    Use Spawn Probability=false
   ) {module=SpawnRate002}
   /SetVariables(
    Emitter.EmissiveColor:LinearColor=<rgba 0.0125 0.0162 0.0208 0.0118>
    Emitter.WetColor:LinearColor=<rgba 0.0285 0.0307 0.0312 0.2251>
    Emitter.WetColorSide:LinearColor=<rgba 0.0407 0.061 0.0781 0.2251>
   ) {isSetParameters=true}
  ]
  particleSpawnScript [
   /SetVariables(Particles.VisibilityTag=&SpawnGroup()) {isSetParameters=true}
   &InitializeParticle(Color Mode=::NewEnumerator0, Lifetime=1.6, UsePositionOffset=false, Write Material Random=true)
   &ShapeLocation(
    Box Size=&MergeHelper_VectorSwitch(0=<30 25 5>, 1=<10 25 5>, 2=<10 25 5>, 3=?, Index=&SpawnGroup(), Num=::NewEnumerator1)
    MaskSpawnGroup=false
    Non Uniform Scale=&MakeVector(
     X=&FloatFromCurve(
      FloatCurve=DataInterface() {propertyValues {Curve {keys [{arriveTangent=6.5676, interpMode=RCIM_Cubic, leaveTangent=6.5676, tangentMode=RCTM_User, time=0, value=0.427}, <at 1 3.29 RCIM_Cubic>]}}}
      CurveIndex=$Emitter.NormalizedLoopAge
     )
     Y=1
     Z=1
    )
    Offset=&MakeVector(
     X=&MergeHelper_FloatSwitch(
      0=0
      1=&RandomRangeFloat(Minimum=20, Maximum=45)
      2=&RandomRangeFloat(Minimum=-20, Maximum=-45)
      3=?
      Index=&SpawnGroup()
      Num=::NewEnumerator1
     )
     Y=&FloatFromCurve(
      FloatCurve=DataInterface() {propertyValues {Curve {keys [{arriveTangent=2.6275, interpMode=RCIM_Cubic, leaveTangent=2.6275, tangentMode=RCTM_User, time=0, value=0}, <at 0.5 1 RCIM_Cubic>]}}}
      CurveIndex=$Emitter.NormalizedLoopAge
      Scale Curve=-300
     )
     Z=0
    )
    Offset Mode=::NewEnumerator0
    Override Seed=false
    Shape Primitive=::NewEnumerator2
    Surface Only=false
    UseOwnerScale=true
   )
   /SetVariables(Particles.DecalSize:Vector3f=&MergeHelper_VectorSwitch(0=<15 45 70>, 1=<15 32 70>, 2=<15 32 70>, 3=?, Index=&SpawnGroup(), Num=::NewEnumerator1)) {isSetParameters=true}
  ]
  particleUpdateScript [
   &ParticleState()
   /SetVariables(Particles.DecalColor:LinearColor=&MakeLinearColorFromVectorAndFloat(
     "Vector (RGB)"=&MakeVector(
      X=&FloatFromCurve(FloatCurve=DataInterface() {propertyValues {Curve {keys [{arriveTangent=2.6346, interpMode=RCIM_Cubic, leaveTangent=2.6346, tangentMode=RCTM_User, time=0, value=-1}, <at 1 0>]}}})
      Y=0
      Z=0
     )
     "Float (Alpha)"=&FloatFromCurve(
      FloatCurve=DataInterface() {propertyValues {Curve {
         defaultValue=1
         keys [
          {arriveTangent=5.9515, interpMode=RCIM_Cubic, leaveTangent=5.9515, tangentMode=RCTM_User, time=0, value=0}
          {interpMode=RCIM_Cubic, tangentMode=RCTM_User, time=0.0975, value=1}
          {arriveTangentWeight=0.2201, interpMode=RCIM_Cubic, leaveTangentWeight=0.2332, tangentMode=RCTM_User, time=0.5259, value=1}
          {arriveTangent=0.086, interpMode=RCIM_Cubic, leaveTangent=0.086, tangentMode=RCTM_User, time=1, value=0}
         ]
        }}}
      Scale Curve=0.2
     )
    )) {isSetParameters=true}
  ]
  rendererClasses [&NiagaraDecalRendererProperties]
  renderers [
   {rendererClass=&NiagaraDecalRendererProperties, rendererIndex=0}
   {rendererClass=&NiagaraDecalRendererProperties, rendererIndex=1}
   {rendererClass=&NiagaraDecalRendererProperties, rendererIndex=2}
  ]
 }
}
summary {
 systemName=NS_Emote_BeachCavalier_Waves
 emitters [
  {emitterName=Wakes_Moving, rendererClasses [&NiagaraMeshRendererProperties]}
  {emitterName=Wakes_Moving_Edge, rendererClasses [&NiagaraMeshRendererProperties]}
  {emitterName=Wakes_Moving_Overlay, rendererClasses [&NiagaraMeshRendererProperties]}
  {emitterName=Splashes, rendererClasses [&NiagaraSpriteRendererProperties]}
  {emitterName=Wetmap, rendererClasses [&NiagaraDecalRendererProperties]}
 ]
 userVariables [
  User.bIsFrontEnd {type=NiagaraBool}
  User.bIsFrontEndPreview {type=NiagaraBool}
 ]
}
